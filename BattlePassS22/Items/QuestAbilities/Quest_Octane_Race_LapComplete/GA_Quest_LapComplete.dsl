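; Gameplay ability for a Battle Pass S22 Octane race lap-completion quest, currently containing only debug output rather than any quest progress logic.
; On activation it gathers Creative score manager actors without using them, then prints "Checkpoint!" and listens repeatedly for Minigame.Stat.Score gameplay events.
; Each received score event prints "Points!".
; The ability ends itself if passive activation fails.
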(event Ability|EventActivateAbility
  (|CallParentFunction)
  (bind _outactors (Actor|GetAllActorsOfClass "/Game/Creative/Devices/ScoreManager/BP_Creative_ScoreManager.BP_Creative_ScoreManager_C"))
  (Ability|Tasks|WaitGameplayEvent "(TagName=\"Minigame.Stat.Score\")" 0 false false
    (:then
      (Development|PrintText "INVTEXT(\"Checkpoint!\")"))
    (:EventReceived
      (Development|PrintText "INVTEXT(\"Points!\")"))))

(event Ability|EventFailedToActivatePassiveAbility
  (params (ActorInfo :type Gameplay%20Ability%20Actor%20Info%20Structure))
  (|CallParentFunction ActorInfo)
  (Ability|EndAbility))
